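; AI decision Blueprint for humanoids that becomes eligible while the character is DBNO (down but not out).
; Its EvaluateStartingConditions override combines the parent class result with a boolean read from the behavior tree owner's blackboard under the DBNO key.
; The decision can start only when both are true.
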(fn EvaluateStartingConditions
  (params (OwnerComp :type Behavior%20Tree%20Component%20Object%20Reference))
  (bind _returnvalue (|CallParentFunction OwnerComp))
  (return (and _returnvalue (Blackboard|GetBool (Variables|Default|GetDBNOKey) OwnerComp))))
